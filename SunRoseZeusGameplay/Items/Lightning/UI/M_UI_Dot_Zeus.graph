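# Translucent UI material for the Zeus lightning item's targeting dot, combining a center dot with a range targeting ring.
# Both shapes come from MF_UI_Box on UVs rotated 0.375 of a turn.
# RingSize scales the outline ring, and Targeting Color blends it from zero-alpha white to red at reduced opacity.
# The dot blends hard and soft SmoothStep masks of the box SDF, then adds to the ring.
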Material /SunRoseZeusGameplay/Items/Lightning/UI/M_UI_Dot_Zeus
graph {
 properties {blendMode=BLEND_Translucent, materialDomain=MD_UI, pixelDepthOffsetMode=PDOM_Legacy, usedShadingModels=Default Lit}
 nodes [
  Add(A=NamedRerouteUsage(), B=NamedRerouteUsage())
  n1 = MaterialFunctionCall(
   UVs=NamedRerouteDeclaration(None=MaterialFunctionCall() {materialFunction=&GetUserInterfaceUV}.Normalized UV) {name=Uvs, nodeColor=<rgba 0.1 0.1 0.1 1>}
   "Rotation Angle (0-1)"=Constant() {r=0.375}
  ) {materialFunction=&CustomRotator}
  n2 = LinearInterpolate(Alpha=ScalarParameter() {parameterName=RingSize, sliderMax=1}) {constB=0.07}
  n3 = ScalarParameter() {defaultValue=0.0075, parameterName=RingThickness}
  n4 = ScalarParameter() {parameterName=EdgeGlowMin}
  n5 = ScalarParameter() {defaultValue=0.015, parameterName=EdgeGlowMax}
  n6 = Multiply(
   A=LinearInterpolate(
    A=Constant4Vector() {constant=<rgba 1 1 1 0>}.RGBA
    B=Constant4Vector() {constant=<rgba 0.9253 0 0 1>}.RGBA
    Alpha=ScalarParameter() {parameterName=Targeting Color}
   ) {constAlpha=1}
   B=Multiply(A=MaterialFunctionCall(UVs=@n1.Rotated Values, Size=@n2, CornerRoundness=@n4, OutlineThickness=@n3, GlowMin=@n4, GlowMax=@n5) {desc=Range Targeting Ring, materialFunction=&MF_UI_Box}.Outline) {constB=0.75, desc=Ring Opacity}
  )
  n7 = MaterialFunctionCall(
   UVs=@n1.Rotated Values
   Size=ScalarParameter() {defaultValue=0.025, parameterName=InnerDotSize}
   CornerRoundness=@n4
   OutlineThickness=Constant()
   GlowMin=@n4
   GlowMax=@n5
  ) {desc=Center Dot, materialFunction=&MF_UI_Box}
  n8 = SmoothStep(Value=@n7.SDF) {constMax=0.01}
  n9 = LinearInterpolate(
   A=NamedRerouteDeclaration(None=OneMinus(None=@n8)) {name=Solid Dot, nodeColor=<rgba 0 1 0.7891 1>}
   B=OneMinus(None=SmoothStep(Value=@n7.SDF) {constMax=0.04, constMin=0.01})
  )
  NamedRerouteDeclaration(None=Add(
    A=NamedRerouteDeclaration(None=@n6) {name=DistanceRing, nodeColor=<rgba 1 0.4219 0 1>}
    B=NamedRerouteDeclaration(None=@n9) {name=Inner Dot, nodeColor=<rgba 0.1 0.1 0.1 1>}
   ) {constB=0}) {name=Dot+DistanceRing, nodeColor=<rgba 0 0 0 1>}
 ]
}
