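# Translucent UI material for the Zeus lightning reticle, showing the timing of three throws on a circular progress ring.
# Throw keyframes divided by total animation time give angular placements for rotated lightning pips and radial segment masks that cut the ring.
# An angle gradient against Progress masks the ring, while per-throw Windup and Fired parameters shift each pip's offset.
# A distanceMorph parameter blends pip texture channels, and smoothsteps derive base and glow from the masks.
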Material /SunRoseZeusGameplay/Items/Lightning/UI/M_UI_Dynamic_Timing_Zeus
graph {
 properties {blendMode=BLEND_Translucent, materialDomain=MD_UI, pixelDepthOffsetMode=PDOM_Legacy, usedShadingModels=Default Lit}
 nodes [
  n1 = NamedRerouteUsage()
  LinearInterpolate(A=Constant4Vector() {constant=<rgba 0 0 0 1>}.RGBA, B=Constant4Vector() {constant=<rgba 1 1 1 0>}.RGBA, Alpha=Ceil(None=Saturate(None=@n1)))
  Multiply(A=Add(A=NamedRerouteUsage(), B=NamedRerouteUsage())) {constB=0.85}
  NamedRerouteDeclaration(None=MaterialFunctionCall() {materialFunction=&GetUserInterfaceUV}.Normalized UV) {name=Uvs, nodeColor=<rgba 0.1 0.1 0.1 1>}
  n2 = TextureSample(UVs=MaterialFunctionCall(UVs=NamedRerouteUsage(), Texture Scale=ScalarParameter() {defaultValue=0.275, parameterName=Secondary Pip Size}) {materialFunction=&ScaleUVsByCenter}.UVs) {samplerSource=SSM_Clamp_WorldGroupSettings, texture=/SunRoseZeusGameplay/Items/Lightning/UI/Images/T_UI_Reticle_ZeusLightningPips}
  NamedRerouteDeclaration(None=LinearInterpolate(
    A=ComponentMask(None=@n2.RGB) {b=true}
    B=ComponentMask(None=@n2.RGB) {g=true}
    Alpha=ScalarParameter() {parameterName=distanceMorph, sliderMax=1}
   )) {name=T1 Texture, desc=Thunder 1 Texture, nodeColor=<rgba 0 0 0 1>}
  n3 = NamedRerouteUsage()
  n4 = Multiply(A=Add(A=NamedRerouteUsage(), B=NamedRerouteUsage())) {constB=-1}
  NamedRerouteDeclaration(None=MaterialFunctionCall!(
    UV=MaterialFunctionCall(UVs=@n3, "Rotation Angle (0-1)"=@n4) {materialFunction=&CustomRotator}.Rotated Values
    Segment Count=Constant() {r=1}
    Segment Spacing=Constant() {r=0.96}
    Fixed Spacing
   ) {materialFunction=/FortUILibrary/MaterialFunctions/Basics/MF_UI_RadialSegments}.Segmented Triangles Fill) {name=T1 Angle Segment, desc=Thunder 1 Segment Mask, nodeColor=<rgba 0.1 0.1 0.1 1>}
  n5 = NamedRerouteUsage()
  n6 = Multiply(A=Add(A=NamedRerouteUsage(), B=NamedRerouteUsage())) {constB=-1}
  NamedRerouteDeclaration(None=MaterialFunctionCall(
    UV=MaterialFunctionCall(UVs=@n5, "Rotation Angle (0-1)"=@n6) {materialFunction=&CustomRotator}.Rotated Values
    Segment Count=Constant() {r=1}
    Segment Spacing=Constant() {r=0.96}
   ) {materialFunction=&MF_UI_RadialSegments}.Segmented Triangles Fill) {name=T2 Angle Segment, desc=Thunder 2 Segment Mask, nodeColor=<rgba 0.1 0.1 0.1 1>}
  n7 = NamedRerouteUsage()
  n8 = NamedRerouteUsage()
  NamedRerouteDeclaration(None=MaterialFunctionCall(
    UV=MaterialFunctionCall(UVs=@n7, "Rotation Angle (0-1)"=@n8) {materialFunction=&CustomRotator}.Rotated Values
    Segment Count=Constant() {r=1}
    Segment Spacing=Constant() {r=0.95}
   ) {materialFunction=&MF_UI_RadialSegments}.Segmented Triangles Fill) {name=T3 Angle Segment, desc=Thunder 3 Segment Mask, nodeColor=<rgba 0.1 0.1 0.1 1>}
  n9 = NamedRerouteUsage()
  n10 = NamedRerouteUsage()
  NamedRerouteDeclaration(None=Add(A=Add(A=@n9, B=@n10), B=NamedRerouteUsage())) {name=Throw Segments, nodeColor=<rgba 0.1 0.1 0.1 1>}
  NamedRerouteDeclaration(None=Subtract(A=NamedRerouteUsage(), B=NamedRerouteUsage())) {name=CIrcle Mask with Throw Segments, nodeColor=<rgba 0.1 0.1 0.1 1>}
  n11 = NamedRerouteUsage()
  n12 = NamedRerouteUsage()
  NamedRerouteDeclaration(None=Add(A=Add(A=@n11, B=@n12), B=NamedRerouteUsage())) {name=Thunder Dots, nodeColor=<rgba 0.1 0.1 0.1 1>}
  n13 = MaterialFunctionCall!(UVs=NamedRerouteUsage(), Rotation=Add(A=NamedRerouteUsage()) {constB=-0.03}) {materialFunction=/FortUILibrary/MaterialFunctions/Gradients/MF_UI_AngleGradient}
  n14 = Saturate(None=Subtract(A=OneMinus(None=@n13.Result), B=ScalarParameter() {parameterName=Progress, sliderMax=1}))
  NamedRerouteDeclaration(None=OneMinus(None=Ceil(None=@n14))) {name=Progress Mask, nodeColor=<rgba 0.1 0.1 0.1 1>}
  n15 = SmoothStep(Value=NamedRerouteUsage()) {constMax=0.5, constMin=0.25}
  n16 = NamedRerouteUsage()
  NamedRerouteDeclaration(None=Multiply(A=Multiply(A=@n15, B=@n16)) {constB=0.6}) {name=Progress Circle, nodeColor=<rgba 0.1 0.1 0.1 1>}
  n17 = NamedRerouteUsage()
  n18 = LinearInterpolate(
   A=NamedRerouteDeclaration(None=SmoothStep(Value=@n17) {constMax=0.55, constMin=0.355}) {name=Circle Base, nodeColor=<rgba 0 0 0 1>}
   B=NamedRerouteDeclaration(None=SmoothStep(Value=@n17) {constMax=0.3, constMin=0.01}) {name=Circle OuterGlow, nodeColor=<rgba 0 0 0 1>}
  ) {constAlpha=0.4}
  n19 = ScalarParameter() {defaultValue=0.4, parameterName=BGOpacity}
  n20 = NamedRerouteUsage()
  n21 = LinearInterpolate(
   A=NamedRerouteDeclaration(None=SmoothStep(Value=@n20) {constMax=0.25, constMin=0.15}) {name=Thunder Base, nodeColor=<rgba 0 0 0 1>}
   B=NamedRerouteDeclaration(None=SmoothStep(Value=@n20) {constMax=0.3, constMin=0.001}) {name=Thunder OuterGlow, nodeColor=<rgba 0 0 0 1>}
  ) {constAlpha=0.4}
  n22 = ScalarParameter() {defaultValue=0.8, parameterName=Pip Opacity}
  NamedRerouteDeclaration(None=Add(A=Multiply(A=@n18, B=@n19), B=Multiply(A=@n21, B=@n22))) {name=Base + Thunder + Glow, nodeColor=<rgba 0.1 0.1 0.1 1>}
  n23 = NamedRerouteUsage()
  n24 = Multiply(A=Add(A=NamedRerouteUsage(), B=NamedRerouteUsage())) {constB=-1}
  n25 = Constant()
  n26 = LinearInterpolate(
   A=LinearInterpolate(Alpha=ScalarParameter() {parameterName=T1 Windup %, sliderMax=1}) {constA=0.26, constB=0.35}
   Alpha=ScalarParameter() {parameterName=T1 Fired %, sliderMax=1}
  ) {constA=0.23, constB=0.23}
  NamedRerouteDeclaration(None=Add(
    A=MaterialFunctionCall(UVs=@n23, "Rotation Angle (0-1)"=@n24) {materialFunction=&CustomRotator}.Rotated Values
    B=AppendVector(A=@n25, B=@n26)
   )) {name=T1 UVs, desc=Thunder 1 UV Placement, nodeColor=<rgba 0.1 0.1 0.1 1>}
  n27 = NamedRerouteUsage()
  n28 = Multiply(A=Add(A=NamedRerouteUsage(), B=NamedRerouteUsage())) {constB=-1}
  n29 = Constant()
  n30 = LinearInterpolate(
   A=LinearInterpolate(Alpha=ScalarParameter() {parameterName=T2 Windup %, sliderMax=1}) {constA=0.26, constB=0.34}
   Alpha=ScalarParameter() {parameterName=T2 Fired %, sliderMax=1}
  ) {constA=0.23, constB=0.22}
  NamedRerouteDeclaration(None=Add(
    A=MaterialFunctionCall(UVs=@n27, "Rotation Angle (0-1)"=@n28) {materialFunction=&CustomRotator}.Rotated Values
    B=AppendVector(A=@n29, B=@n30)
   )) {name=T2 UVs, desc=Thunder 2 UV Placement, nodeColor=<rgba 0.1 0.1 0.1 1>}
  n31 = TextureSample(UVs=MaterialFunctionCall(UVs=NamedRerouteUsage(), Texture Scale=ScalarParameter() {defaultValue=0.275, parameterName=Secondary Pip Size}) {materialFunction=&ScaleUVsByCenter}.UVs) {samplerSource=SSM_Clamp_WorldGroupSettings, texture=&T_UI_Reticle_ZeusLightningPips}
  NamedRerouteDeclaration(None=LinearInterpolate(
    A=ComponentMask(None=@n31.RGB) {b=true}
    B=ComponentMask(None=@n31.RGB) {g=true}
    Alpha=ScalarParameter() {parameterName=distanceMorph, sliderMax=1}
   )) {name=T2 Texture, desc=Thunder 2 Texture, nodeColor=<rgba 0 0 0 1>}
  NamedRerouteDeclaration(None=Divide(A=ScalarParameter() {defaultValue=92, parameterName=Throw1Keyframe}, B=ScalarParameter() {defaultValue=270, parameterName=TotalAnimTime})) {name=T1 Progress %, desc=Thunder 1 Keyframe Placement, nodeColor=<rgba 0.1 0.1 0.1 1>}
  n32 = NamedRerouteUsage()
  n33 = NamedRerouteUsage()
  n34 = Constant()
  n35 = LinearInterpolate(
   A=LinearInterpolate(Alpha=ScalarParameter() {parameterName=T3 Windup %, sliderMax=1}) {constA=0.285, constB=0.38}
   Alpha=ScalarParameter() {parameterName=T3 Fired %, sliderMax=1}
  ) {constA=0.23, constB=0.23}
  NamedRerouteDeclaration(None=Add(
    A=MaterialFunctionCall(UVs=@n32, "Rotation Angle (0-1)"=@n33) {materialFunction=&CustomRotator}.Rotated Values
    B=AppendVector(A=@n34, B=@n35)
   )) {name=T3 UVs, desc=Thunder 3 UV Placement, nodeColor=<rgba 0.1 0.1 0.1 1>}
  NamedRerouteDeclaration(None=Max(A=NamedRerouteUsage(), B=NamedRerouteUsage())) {name=Base + Thunder No Glow, nodeColor=<rgba 0 0 0 1>}
  NamedRerouteDeclaration(None=Divide(A=ScalarParameter() {defaultValue=125, parameterName=Throw2Keyframe}, B=ScalarParameter() {defaultValue=270, parameterName=TotalAnimTime})) {name=T2 Progress %, desc=Thunder 2 Keyframe Placement, nodeColor=<rgba 0.1 0.1 0.1 1>}
  NamedRerouteDeclaration(None=Divide(A=ScalarParameter() {defaultValue=270, parameterName=Throw3Keframe}, B=ScalarParameter() {defaultValue=270, parameterName=TotalAnimTime})) {name=T3 Progress %, desc=Thunder 3 Keyframe Placement, nodeColor=<rgba 0.1 0.1 0.1 1>}
  NamedRerouteDeclaration(None=ScalarParameter() {defaultValue=0.0356, parameterName=SegmentArc}) {name=T3+SegmentArc, nodeColor=<rgba 0.1 0.1 0.1 1>}
  n36 = MaterialFunctionCall(
   UVs=Add(A=Constant2Vector() {g=0.01}, B=NamedRerouteUsage())
   Texture Scale=ScalarParameter() {defaultValue=0.55, parameterName=SizeOfReticle, sliderMax=1}
  ) {materialFunction=&ScaleUVsByCenter}
  NamedRerouteDeclaration(None=ComponentMask(None=TextureSample(UVs=@n36.UVs) {samplerSource=SSM_Clamp_WorldGroupSettings, texture=/SunRoseZeusGameplay/Items/Lightning/UI/Images/T_UI_Reticle_ZeusLightning}.RGB) {r=true}) {name=Texture Circle Mask, nodeColor=<rgba 0 0 0 1>}
  n37 = TextureSample(UVs=MaterialFunctionCall(UVs=NamedRerouteUsage(), Texture Scale=ScalarParameter() {defaultValue=0.3, parameterName=Pip Size}) {materialFunction=&ScaleUVsByCenter}.UVs) {samplerSource=SSM_Clamp_WorldGroupSettings, texture=&T_UI_Reticle_ZeusLightning}
  NamedRerouteDeclaration(None=LinearInterpolate(
    A=ComponentMask(None=@n37.RGB) {b=true}
    B=ComponentMask(None=@n37.RGB) {g=true}
    Alpha=ScalarParameter() {parameterName=distanceMorph, sliderMax=1}
   )) {name=T3 Texture, desc=Thunder 3 Texture, nodeColor=<rgba 0 0 0 1>}
 ]
}
