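# Reticle extension widget for the SunRose Zeus lightning weapon, showing throw progress, target acquisition and cooldown through a material-driven reticle and center dot.
# Construct binds to the player controller's actor-under-reticle and weapon's cooldown delegates, snapping cooldown animations to current state; Destruct unbinds.
# Firing plays Attack, animating three staged windup and fired material parameters.
# Target changes play ActorInRange forward or reverse off cooldown; cooldown changes dim, shrink and rotate elements, or restore them.
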WidgetBlueprint /SunRoseZeusGameplay/Items/Lightning/UI/WBP_UI_SunRoseZeus_RetExtension
animations [
 Attack {
  binding_count=1
  length_seconds=4.5333
  object=&self:Attack
  sequence {
   bindings [Reticle {
     id=#1
     possessed_class=Image
     tracks [{
       type=MovieSceneWidgetMaterialTrack
       sections [{
         type=MovieSceneParameterSection
         end_frame=136
         start_frame=0
         channels [
          Progress {type=MovieSceneScriptingFloatChannel, keys [[0,0],[135,1]]}
          T1ThrownSwitch {type=MovieSceneScriptingFloatChannel, keys [[0,0],[50,1]]}
          T2ThrownSwitch {type=MovieSceneScriptingFloatChannel, keys [[0,0],[68,1]]}
          T3ThrownSwitch {type=MovieSceneScriptingFloatChannel, keys [[0,0],[135,1]]}
          IsReadyToFireSwitch {type=MovieSceneScriptingFloatChannel, const [0,2]}
          T1 Fired % {type=MovieSceneScriptingFloatChannel, keys [[46,0],[50,1]]}
          T1 Windup % {type=MovieSceneScriptingFloatChannel, keys [[26,0],[46,1]]}
          T2 Fired % {type=MovieSceneScriptingFloatChannel, keys [[64,0],[68,1]]}
          T2 Windup % {type=MovieSceneScriptingFloatChannel, keys [[44,0],[64,1]]}
          T3 Fired % {type=MovieSceneScriptingFloatChannel, keys [[131,0],[135,1]]}
          T3 Windup % {type=MovieSceneScriptingFloatChannel, keys [[101,0],[131,1]]}
         ]
        }]
      }]
    }]
   display_rate {numerator=30}
  }
 }
 ActorInRange {
  binding_count=2
  length_seconds=0.2336
  object=&self:ActorInRange
  sequence {
   bindings [
    Reticle {
     id=#2
     possessed_class=Image
     tracks [{
       type=MovieSceneWidgetMaterialTrack
       sections [{
         type=MovieSceneParameterSection
         end_frame=7
         start_frame=0
         channels [
          distanceMorph {type=MovieSceneScriptingFloatChannel, keys [[0,0],[5,1]]}
          Pip Opacity {type=MovieSceneScriptingFloatChannel, keys [[0,0.5],[6,0.85]]}
          BGOpacity {type=MovieSceneScriptingFloatChannel, keys [[0,0.4],[6,0.6]]}
         ]
        }]
      }]
    }
    Dot {
     id=#3
     possessed_class=Image
     tracks [
      {type=MovieSceneWidgetMaterialTrack, sections [{type=MovieSceneParameterSection}]}
      {
       type=MovieSceneColorTrack
       display_name=Color and Opacity
       sections [{
         type=MovieSceneColorSection
         end_frame=6
         start_frame=0
         channels [
          Color.R {type=MovieSceneScriptingFloatChannel, const [1,1]}
          Color.G {type=MovieSceneScriptingFloatChannel, const [1,1]}
          Color.B {type=MovieSceneScriptingFloatChannel, const [1,1]}
          Color.A {type=MovieSceneScriptingFloatChannel, keys [[0,0.6],[6,1]]}
         ]
        }]
      }
     ]
    }
   ]
   display_rate {numerator=30}
  }
 }
 OffCooldown {
  binding_count=2
  length_seconds=0.2336
  object=&self:OffCooldown
  sequence {
   bindings [
    Dot {
     id=#4
     possessed_class=Image
     tracks [
      {
       type=MovieSceneColorTrack
       display_name=Color and Opacity
       sections [{
         type=MovieSceneColorSection
         channels [
          Color.R {type=MovieSceneScriptingFloatChannel, const [1,1]}
          Color.G {type=MovieSceneScriptingFloatChannel, const [1,1]}
          Color.B {type=MovieSceneScriptingFloatChannel, const [1,1]}
          Color.A {type=MovieSceneScriptingFloatChannel, keys [[0,0.5],[5,1]]}
         ]
        }]
      }
      {
       type=MovieScene2DTransformTrack
       sections [{
         type=MovieScene2DTransformSection
         channels [
          Translation.X {type=MovieSceneScriptingFloatChannel, const [0,1]}
          Translation.Y {type=MovieSceneScriptingFloatChannel, const [0,1]}
          Angle {type=MovieSceneScriptingFloatChannel, keys [[0,45],[5,0]]}
          Scale.X {type=MovieSceneScriptingFloatChannel, const [1,1]}
          Scale.Y {type=MovieSceneScriptingFloatChannel, const [1,1]}
          Shear.X {type=MovieSceneScriptingFloatChannel, const [0,1]}
          Shear.Y {type=MovieSceneScriptingFloatChannel, const [0,1]}
         ]
        }]
      }
     ]
    }
    Reticle {
     id=#5
     possessed_class=Image
     tracks [
      {
       type=MovieSceneColorTrack
       display_name=Color and Opacity
       sections [{
         type=MovieSceneColorSection
         end_frame=7
         start_frame=0
         channels [
          Color.R {type=MovieSceneScriptingFloatChannel, const [1,1]}
          Color.G {type=MovieSceneScriptingFloatChannel, const [1,1]}
          Color.B {type=MovieSceneScriptingFloatChannel, const [1,1]}
          Color.A {type=MovieSceneScriptingFloatChannel, keys [[0,0.5],[7.008,1]]}
         ]
        }]
      }
      {
       type=MovieScene2DTransformTrack
       sections [{
         type=MovieScene2DTransformSection
         end_frame=7
         start_frame=0
         channels [
          Translation.X {type=MovieSceneScriptingFloatChannel, const [0,1]}
          Translation.Y {type=MovieSceneScriptingFloatChannel, const [0,1]}
          Angle {type=MovieSceneScriptingFloatChannel}
          Scale.X {type=MovieSceneScriptingFloatChannel, keys [[0,0.85],[5,1]]}
          Scale.Y {type=MovieSceneScriptingFloatChannel, keys [[0,0.85],[5,1]]}
          Shear.X {type=MovieSceneScriptingFloatChannel, const [0,1]}
          Shear.Y {type=MovieSceneScriptingFloatChannel, const [0,1]}
         ]
        }]
      }
      {
       type=MovieSceneWidgetMaterialTrack
       sections [{
         type=MovieSceneParameterSection
         end_frame=7
         start_frame=0
         channels [
          T1 Fired % {type=MovieSceneScriptingFloatChannel, keys [[0,1],[3,0]]}
          T2 Fired % {type=MovieSceneScriptingFloatChannel, keys [[0,1],[5,0]]}
          T3 Fired % {type=MovieSceneScriptingFloatChannel, keys [[0,1],[7.008,0]]}
          Progress {type=MovieSceneScriptingFloatChannel, const [0,2]}
         ]
        }]
      }
     ]
    }
   ]
   display_rate {numerator=30}
  }
 }
 OnCooldown {
  binding_count=2
  length_seconds=0.1667
  object=&self:OnCooldown
  sequence {
   bindings [
    Reticle {
     id=#6
     possessed_class=Image
     tracks [
      {
       type=MovieSceneWidgetMaterialTrack
       sections [{
         type=MovieSceneParameterSection
         end_frame=5
         start_frame=0
         channels [
          T1 Fired % {type=MovieSceneScriptingFloatChannel, const [1,1]}
          T2 Fired % {type=MovieSceneScriptingFloatChannel, const [1,1]}
          T3 Fired % {type=MovieSceneScriptingFloatChannel, const [1,1]}
         ]
        }]
      }
      {
       type=MovieSceneColorTrack
       display_name=Color and Opacity
       sections [{
         type=MovieSceneColorSection
         end_frame=5
         start_frame=0
         channels [
          Color.R {type=MovieSceneScriptingFloatChannel, const [1,1]}
          Color.G {type=MovieSceneScriptingFloatChannel, const [1,1]}
          Color.B {type=MovieSceneScriptingFloatChannel, const [1,1]}
          Color.A {type=MovieSceneScriptingFloatChannel, keys [[0,1],[5,0.5]]}
         ]
        }]
      }
      {
       type=MovieScene2DTransformTrack
       sections [{
         type=MovieScene2DTransformSection
         end_frame=5
         start_frame=0
         channels [
          Translation.X {type=MovieSceneScriptingFloatChannel, const [0,1]}
          Translation.Y {type=MovieSceneScriptingFloatChannel, const [0,1]}
          Angle {type=MovieSceneScriptingFloatChannel}
          Scale.X {type=MovieSceneScriptingFloatChannel, keys [[0,1],[5,0.85]]}
          Scale.Y {type=MovieSceneScriptingFloatChannel, keys [[0,1],[5,0.85]]}
          Shear.X {type=MovieSceneScriptingFloatChannel, const [0,1]}
          Shear.Y {type=MovieSceneScriptingFloatChannel, const [0,1]}
         ]
        }]
      }
     ]
    }
    Dot {
     id=#7
     possessed_class=Image
     tracks [
      {
       type=MovieSceneColorTrack
       display_name=Color and Opacity
       sections [{
         type=MovieSceneColorSection
         end_frame=5
         start_frame=0
         channels [
          Color.R {type=MovieSceneScriptingFloatChannel, const [1,1]}
          Color.G {type=MovieSceneScriptingFloatChannel, const [1,1]}
          Color.B {type=MovieSceneScriptingFloatChannel, const [1,1]}
          Color.A {type=MovieSceneScriptingFloatChannel, keys [[0,1],[5,0.25]]}
         ]
        }]
      }
      {
       type=MovieScene2DTransformTrack
       sections [{
         type=MovieScene2DTransformSection
         end_frame=5
         start_frame=0
         channels [
          Translation.X {type=MovieSceneScriptingFloatChannel, const [0,1]}
          Translation.Y {type=MovieSceneScriptingFloatChannel, const [0,1]}
          Angle {type=MovieSceneScriptingFloatChannel, keys [[0,0],[5,45]]}
          Scale.X {type=MovieSceneScriptingFloatChannel, const [1,1]}
          Scale.Y {type=MovieSceneScriptingFloatChannel, const [1,1]}
          Shear.X {type=MovieSceneScriptingFloatChannel, const [0,1]}
          Shear.Y {type=MovieSceneScriptingFloatChannel, const [0,1]}
         ]
        }]
      }
     ]
    }
   ]
   display_rate {numerator=30}
  }
 }
]
graphs [
 EventGraph {nodes [
   n1 = On Actor Under Reticle Changed!() -> (OutputDelegate, then, ActorUnderReticle:Actor Object Reference) {is=title K2Node_CustomEvent}
   > n2 = Sequence()
   n3 = Read_IsOnCooldown!(self:B Athena Sun Rose Zeus Lightning Object Reference="Get  Sun Rose Zeus Lightning"() -> (" Sun Rose Zeus Lightning":B Athena Sun Rose Zeus Lightning Object Reference)) -> (IsOnCooldown:bool)
   n4 = IsValid(Object=Get PreviousActorUnderReticle() -> (PreviousActorUnderReticle:Actor Object Reference))
   n5 = IsValid(Object=@n1.ActorUnderReticle)
   > Branch(Condition=AND Boolean(A="Not Equal (Boolean)"(A=@n5, B=@n4) -> (ReturnValue=~), B=NOT Boolean(A=@n3)))
   > PlayActorInRange!(execute, self, Play Forward:bool=IsValid(Object=@n1.ActorUnderReticle)) -> (then)
   Set PreviousActorUnderReticle(execute=@n2.then_1, PreviousActorUnderReticle:Actor Object Reference=@n1.ActorUnderReticle) -> (Output_Get:Actor Object Reference)
   n6 = On Cooldown Changed!() -> (OutputDelegate, then, Is on Cooldown:bool=~) {is=title K2Node_CustomEvent}
   > n7 = Branch(Condition=@n6.Is on Cooldown)
   > PlayActorInRange(Play Forward=false)
   > PlayAnimation(InAnimation=Get OnCooldown() -> (OnCooldown:Widget Animation Object Reference))
   PlayActorInRange(execute=@n7.else, Play Forward=IsValid(Object=Get PreviousActorUnderReticle()))
   > PlayAnimation(InAnimation=Get OffCooldown() -> (OffCooldown:Widget Animation Object Reference))
   Event Construct()
   > n8 = Cast To FortPlayerController(Object=GetOwningPlayer())
   n9 = Create Event()
   > Bind Event to On Actor Under Reticle Set BP(self:Fort Player Controller Object Reference=@n8, Delegate=@n9) {is=title K2Node_AddDelegate}
   n10 = GetEquippedWeapon!(self:Fort Weapon Reticle Extension Widget Base Object Reference) -> (ReturnValue:Fort Weapon Object Reference)
   > Is Valid(InputObject=@n10)
   > n11 = Cast To B_Athena_SunRose_Zeus_Lightning!(execute, Object=@n10) -> (then, CastFailed, AsB Athena Sun Rose Zeus Lightning:B Athena Sun Rose Zeus Lightning Object Reference) {is=title K2Node_DynamicCast}
   > n12 = "Set  Sun Rose Zeus Lightning"(" Sun Rose Zeus Lightning":B Athena Sun Rose Zeus Lightning Object Reference=@n11) -> (Output_Get:B Athena Sun Rose Zeus Lightning Object Reference)
   > n13 = Branch(Condition=Read_IsOnCooldown(self=@n12))
   Bind Event to On Is Cooldown Changed(execute=[@n18.then, @n13.then], self:B Athena Sun Rose Zeus Lightning Object Reference=@n12, Delegate=Create Event()) {is=title K2Node_AddDelegate}
   > n14 = Branch(Condition=Read_IsOnCooldown(self=@n12))
   n15 = Get OnCooldown()
   > PlayAnimation(InAnimation=@n15, StartAtTime=GetEndTime(self=@n15))
   n16 = Get OffCooldown()
   PlayAnimation(execute=@n14.else, InAnimation=@n16, StartAtTime=GetEndTime(self=@n16))
   n17 = Get Attack() -> (Attack:Widget Animation Object Reference)
   n18 = PlayAnimation(execute=@n13.else, InAnimation=@n17, StartAtTime=GetEndTime(self=@n17), PlayMode=Reverse)
   Event Destruct()
   > n19 = Cast To FortPlayerController(Object=GetOwningPlayer())
   > n20 = Unbind Event from On Actor Under Reticle Set BP(self:Fort Player Controller Object Reference=@n19, Delegate=@n9) {is=title K2Node_RemoveDelegate}
   n21 = Get!(execute:Exec=[@n20.then, @n19.CastFailed]) -> (then:Exec, else:Exec, " Sun Rose Zeus Lightning":B Athena Sun Rose Zeus Lightning Object Reference)
   > Unbind Event from On Is Cooldown Changed(self:B Athena Sun Rose Zeus Lightning Object Reference=@n21, Delegate=Create Event()) {is=title K2Node_RemoveDelegate}
   Event OnWeaponSingleFire() {is=title K2Node_Event}
   > "Play Animation with Finished event (legacy)"!(
    execute:Exec
    Widget:User Widget Object Reference=Self-Reference()
    InAnimation:Widget Animation Object Reference=Get Attack()
    StartAtTime:float=0
    NumLoopsToPlay:int=1
    PlayMode:EUMGSequencePlayMode Enum
    PlaybackSpeed:float=1
   ) -> (then:Exec, Result:UMGSequence Player Object Reference, Finished:Exec) {is=title K2Node_PlayAnimation}
  ]}
 PlayActorInRange {nodes [
   n1 = PlayActorInRange!() -> (then, Play Forward:bool=~) {is=title K2Node_FunctionEntry}
   > n2 = Branch(Condition=@n1)
   n3 = Get ActorInRange() -> (ActorInRange:Widget Animation Object Reference)
   > PlayAnimationForward(InAnimation=@n3)
   PlayAnimationReverse(execute=@n2.else, InAnimation=@n3)
  ]}
]
variables [
 " Sun Rose Zeus Lightning" {type=B_Athena_SunRose_Zeus_Lightning_C, default=null}
 PreviousActorUnderReticle {type=Actor, default=null}
 TagsRepliatedTagsGrantedToPlayerWhileActive {type=GameplayTag, default {tagName=Granted.Athena.SunRose.Zeus.Lightning.AbilityActive}}
]
widgets {
 info {
  inheritedWidgetCount=1
  parentClass=/Script/FortniteUI.FortWeaponReticleExtensionWidgetBase
  rootWidgetClass=&Overlay
  widgetCount=4
 }
 widgets [
  Overlay_11 {
   widgetClassPath=&Overlay
   widget=Overlay_11
   properties {renderTransform {angle=0, scale=<0.9 0.9>, shear=<0 0>, translation=<0 0>}}
  }
  Reticle {
   widgetClassPath=&Image
   bIsVariable=true
   parent=Overlay_11
   widget=Reticle
   properties {
    visibility=HitTestInvisible
    brush {
     bIsDynamicallyLoaded=false
     drawAs=Image
     imageType=FullColor
     mirroring=NoMirror
     resourceName=None
     resourceObject=&/M_UI_DT_Zeus_Inst
     tiling=NoTile
     imageSize=<150 150>
     margin=<ltrb 0 0 0 0>
     outlineSettings {
      bUseBrushTransparency=false
      roundingType=HalfHeightRadius
      width=0
      color {colorUseRule=UseColor_Specified, specifiedColor=<rgba 0 0 0 0>}
      cornerRadii=<0 0 0 0>
     }
     tintColor {colorUseRule=UseColor_Specified, specifiedColor=<rgba 1 1 1 1>}
     uVRegion {bIsValid=false, max=<0 0>, min=<0 0>}
    }
   }
   slot {class=OverlaySlot, properties {horizontalAlignment=HAlign_Center, verticalAlignment=VAlign_Center}}
  }
  Dot {
   widgetClassPath=&Image
   bIsVariable=true
   parent=Overlay_11
   widget=Dot
   properties {
    visibility=HitTestInvisible
    brush {
     bIsDynamicallyLoaded=false
     drawAs=Image
     imageType=FullColor
     mirroring=NoMirror
     resourceName=None
     resourceObject=&/MI_UI_Dot_Zeus
     tiling=NoTile
     imageSize=<150 150>
     margin=<ltrb 0 0 0 0>
     outlineSettings {
      bUseBrushTransparency=false
      roundingType=HalfHeightRadius
      width=0
      color {colorUseRule=UseColor_Specified, specifiedColor=<rgba 0 0 0 0>}
      cornerRadii=<0 0 0 0>
     }
     tintColor {colorUseRule=UseColor_Specified, specifiedColor=<rgba 1 1 1 1>}
     uVRegion {bIsValid=false, max=<0 0>, min=<0 0>}
    }
   }
   slot {class=OverlaySlot, properties {horizontalAlignment=HAlign_Center, verticalAlignment=VAlign_Center}}
  }
  TutorialHighlight {widgetClassPath=&TutorialHighlight}
 ]
}
